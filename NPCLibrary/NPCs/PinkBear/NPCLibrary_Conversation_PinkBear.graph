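# Conversation database for the Pink Bear Medic specialist NPC, defining what it says when a player talks to it.
# The entry point leads to one Speech node with a default line and contextual lines chosen by first-ever conversation or per-match conversation-count requirements.
# That node then branches through link nodes to shared NPC Library quest, basic services and Season 37 NPC rescue combat conversations.
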ConversationDatabase /NPCLibrary/NPCs/PinkBear/NPCLibrary_Conversation_PinkBear
entry_points [{tag=Conversation.Entry.PinkBear, destinations [#1]}]
nodes [
 ConversationEntryPointNode_7 {
  id=#1
  class=ConversationEntryPointNode
  outputs [#2]
  properties {entryTag {tagName=Conversation.Entry.PinkBear}}
 }
 Speech_C_4 {
  id=#2
  class=Speech_C
  outputs [#3, #4, #5]
  properties {
   nodeName=""
   generalConfig {
    defaultMessage=Fuzzy. Fierce. Flawless.
    contextualMessages [
     {
      message="Stay strong against the Swarm!"
      contextRequirements [{requirement {class=FortControllerRequirement_IsFirstEverConversationWithNPC}}]
      priority {curve {rowName=Tandem.Speech.Priority.FirstEverInteraction}}
     }
     {
      message="This fight isn't over yet!"
      contextRequirements [{requirement {
         class=FortControllerRequirement_CheckNPCMatchConversationCount
         properties {nPCUniqueIdTag {tagName=AISpawnerData.Type.Specialist.Medic.PinkBear}}
        }}]
      priority {curve {rowName=None}}
     }
     {
      message="In times like these, we all have to be Team Leaders."
      contextRequirements [{requirement {
         class=FortControllerRequirement_CheckNPCMatchConversationCount
         properties {nPCUniqueIdTag {tagName=AISpawnerData.Type.Specialist.Medic.PinkBear}}
        }}]
      priority {curve {rowName=None}}
     }
     {
      message=Fuzzy. Fierce. Flawless.
      contextRequirements [{requirement {
         class=FortControllerRequirement_CheckNPCMatchConversationCount
         properties {nPCUniqueIdTag {tagName=AISpawnerData.Type.Specialist.Medic.PinkBear}}
        }}]
      priority {curve {rowName=None}}
     }
    ]
   }
  }
 }
 ConversationLinkNode_2 {
  id=#5
  class=ConversationLinkNode
  properties {remoteEntryTag {tagName=Conversation.Entry.Season37.NPCRescues.Combat}}
 }
 ConversationLinkNode_0 {id=#4, class=ConversationLinkNode, properties {remoteEntryTag {tagName=Conversation.Entry.BasicServices}}}
 ConversationLinkNode_1 {id=#3, class=ConversationLinkNode, properties {remoteEntryTag {tagName=Conversation.Entry.NPCLibrary.Quests}}}
]
